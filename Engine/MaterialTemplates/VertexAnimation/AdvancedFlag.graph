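# Two-sided lit material template for a wind-animated flag or cloth, with tiled default diffuse and normal textures.
# Large and small sine waves pan through world space along a constant-derived wind vector, plus a slower macro gust, scaled by strength parameters.
# RotateAboutAxis adds wind-driven bending and a height-masked twist around Z.
# Vertex color weights the small waves and the combined offset; its node note assigns red to ripples, green to overall weight.
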Material /Engine/MaterialTemplates/VertexAnimation/AdvancedFlag
graph {
 properties {bUsedWithStaticLighting=true, pixelDepthOffsetMode=PDOM_Legacy, twoSided=true, usedShadingModels=Default Lit}
 nodes [
  n1 = TextureCoordinate() {uTiling=6, vTiling=6}
  MaterialFunctionCall(
   DiffuseColor=TextureSample(UVs=@n1) {texture=/Engine/EngineMaterials/DefaultDiffuse}.RGB
   SpecularColor=Multiply(A=Constant() {r=1}) {constA=1, constB=0.04, desc=FresnelBaseReflectFraction}
  ) {materialFunction=&ConvertFromDiffSpec}
  n2 = VertexColor() {desc=VERTEX COLORS RED is Little Waves GREEN is Overall weight}
  n3 = ComponentMask(None=Constant() {r=0.5}) {b=true, g=true, r=true}
  MaterialExpressionNormalize_0 = Normalize(VectorInput=@n3)
  n4 = Add(A=@MaterialExpressionNormalize_0, B=Constant3Vector() {constant=<rgba 0 0 -1 0>})
  n5 = Multiply(A=WorldPosition().XYZ, B=Constant3Vector() {constant=<rgba 1 1 0.5 0>})
  n6 = Divide(A=@n5, B=ScalarParameter() {defaultValue=150, parameterName=WIND_LgWaveLength})
  n7 = Constant() {r=0.5}
  n8 = Add(A=Multiply(A=Time(), B=Constant() {r=0.5}), B=Add(A=ComponentMask(None=@n7) {r=true}, B=ComponentMask(None=@n7) {g=true}))
  n9 = Multiply(A=ComponentMask(None=Constant() {r=0.5}) {a=true}, B=Constant() {r=0.5})
  n10 = Multiply(A=Multiply(A=Multiply(A=@n8, B=@n9), B=Constant() {r=-1}), B=@n4)
  n11 = Sine(None=Multiply(A=@n4, B=Add(A=@n6, B=@n10) {desc=Pan time through space}) {desc=Cancel out waves not in line with wind direction}) {desc=Make a Wave}
  n12 = ScalarParameter() {defaultValue=2, parameterName=WIND_SmWaveAmp}
  n13 = Add(A=Divide(A=@n5, B=ScalarParameter() {defaultValue=60, parameterName=WIND_SmWaveLength}), B=@n10) {desc=Pan time through space}
  n14 = ConstantBiasScale(None=Sine(None=Multiply(A=@n4, B=@n13) {desc=Cancel out waves not in line with wind direction}) {desc=Make a Wave}) {desc=Go from 0 to 1 instead of -1 to 1}
  n15 = Add(
   A=Multiply(A=ConstantBiasScale(None=@n11) {desc=Go from 0 to 1 instead of -1 to 1}, B=ScalarParameter() {defaultValue=2, parameterName=WIND_LgWaveAmp})
   B=Multiply(A=@n2, B=Multiply(A=@n12, B=@n14))
  )
  n16 = Constant()
  n17 = Add(A=@n15, B=Multiply(A=Distance(A=@n15, B=@n16), B=Constant3Vector() {constant=<rgba 1 1 1 0>}))
  n18 = Multiply(A=@n8, B=@n9)
  n19 = Constant() {r=-0.2342}
  n20 = ConstantBiasScale(None=Sine(None=Multiply(A=@n18, B=@n19)) {desc=Make a Wave})
  n21 = Constant() {r=2}
  n22 = Multiply(A=Multiply(A=Power(Base=@n20, Exp=@n21), B=Constant() {desc=MacroWaveAmplitude, r=10}), B=@n4)
  n23 = Distance(A=@n3, B=Constant())
  n24 = Multiply(
   A=Multiply(A=Add(A=@n17, B=@n22), B=@n23) {desc=Scale the Wave by WindStrength}
   B=ScalarParameter() {defaultValue=2, parameterName=WaveStrength}
  ) {desc=Amplitude}
  n25 = Custom!(A=@MaterialExpressionNormalize_0) {
   code="return float3(A.g,A.r*-1,0);"
   inputs [{inputName=A, input {expression=&self:MaterialExpressionNormalize_0}}]
  }
  n26 = Multiply(A=ScalarParameter() {defaultValue=0.25, parameterName=RotationalAngle}, B=@n23)
  n27 = Multiply(A=@n8, B=Constant() {r=0.25})
  n28 = Multiply(A=@n8, B=Constant() {r=0.01})
  n29 = Multiply(A=ConstantBiasScale(None=Sine(None=@n27)), B=ConstantBiasScale(None=Sine(None=@n28)))
  n30 = RotateAboutAxis(
   NormalizedRotationAxis=AppendVector(A=Normalize(VectorInput=@n25), B=Multiply(A=@n26, B=@n29))
   RotationAngle=Constant3Vector()
   PivotPoint=Constant3Vector()
   Position=WorldPosition().XYZ
  )
  n31 = LinearInterpolate(A=Constant(), B=ScalarParameter() {defaultValue=1, parameterName=WIND_TwistAngle})
  n32 = Add(B=VectorParameter() {parameterName=RotationalAxisPosition, defaultValue=<rgba 0 0 300 1>}.RGB)
  n33 = Constant3Vector() {constant=<rgba 0 0 1 0>}
  n34 = WorldPosition()
  n35 = Power(Base=SphereMask(A=Multiply(A=@n32, B=@n33), B=Multiply(A=@n33, B=@n34.XYZ)) {attenuationRadius=800, hardnessPercent=0}, Exp=Constant() {r=3})
  n36 = RotateAboutAxis(
   NormalizedRotationAxis=AppendVector(A=Constant3Vector() {constant=<rgba 0 0 1 0>}, B=Multiply(A=@n20, B=@n31))
   RotationAngle=LinearInterpolate(A=Constant() {r=-0.5}, B=Constant() {r=1}, Alpha=OneMinus(None=@n35))
   PivotPoint=@n32
   Position=WorldPosition().XYZ
  )
  Multiply(A=@n2, B=Add(A=Add(A=@n24, B=@n30) {desc=wave+Rotation}, B=Multiply(B=@n36)) {desc=+Twist}) {desc=*VertexColor}
  TextureSample(UVs=@n1) {samplerType=SAMPLERTYPE_Normal, texture=/Engine/EngineMaterials/DefaultNormal}
 ]
}
