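# Engine template material for a simple vertex bulge effect, named SimpleBuldge, using Default Lit shading.
# A BuldgeAmount scalar expands to float3 in a Custom node and is scaled by inverted vertex color, so painting masks the bulge.
# Tiled default diffuse, blended toward a constant, and a constant specular feed ConvertFromDiffSpec, alongside a tiled default normal map.
# The graph shows no connections to material outputs, including World Position Offset.
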Material /Engine/MaterialTemplates/VertexAnimation/SimpleBuldge
graph {
 properties {pixelDepthOffsetMode=PDOM_Legacy, usedShadingModels=Default Lit}
 nodes [
  n1 = TextureCoordinate() {uTiling=6, vTiling=6}
  n2 = Constant() {r=0.25}
  MaterialFunctionCall(
   DiffuseColor=LinearInterpolate(A=TextureSample(UVs=@n1) {texture=/Engine/EngineMaterials/DefaultDiffuse}.RGB, B=@n2, Alpha=@n2)
   SpecularColor=Multiply(A=Constant() {r=1}) {constA=1, constB=0.04, desc=FresnelBaseReflectFraction}
  ) {materialFunction=&ConvertFromDiffSpec}
  MaterialExpressionScalarParameter_0 = ScalarParameter() {defaultValue=32, parameterName=BuldgeAmount}
  n3 = Custom!(A=@MaterialExpressionScalarParameter_0) {
   code="return float3(A.r,A.r,A.r);"
   inputs [{inputName=A, input {expression=&self:MaterialExpressionScalarParameter_0}}]
  }
  Multiply(A=OneMinus(None=VertexColor()), B=Multiply(B=@n3))
  TextureSample(UVs=@n1) {samplerType=SAMPLERTYPE_Normal, texture=/Engine/EngineMaterials/DefaultNormal}
 ]
}
