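# Material template for simple rotational vertex animation about a parameter-defined axis, on a Default Lit surface.
# RotateAboutAxis takes world position, the normalized RotationalAxis appended with the static RotationalAngle parameter, and a pivot derived by subtraction using RotationalAxisPosition.
# Tiled engine default diffuse and normal textures provide placeholder shading, with the diffuse blended toward a constant and converted through ConvertFromDiffSpec.
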Material /Engine/MaterialTemplates/VertexAnimation/SimpleRotation
graph {
 properties {pixelDepthOffsetMode=PDOM_Legacy, usedShadingModels=Default Lit}
 nodes [
  n1 = TextureCoordinate() {uTiling=6, vTiling=6}
  n2 = Constant() {r=0.25}
  MaterialFunctionCall(
   DiffuseColor=LinearInterpolate(A=TextureSample(UVs=@n1) {texture=/Engine/EngineMaterials/DefaultDiffuse}.RGB, B=@n2, Alpha=@n2)
   SpecularColor=Multiply(A=Constant() {r=1}) {constA=1, constB=0.04, desc=FresnelBaseReflectFraction}
  ) {materialFunction=&ConvertFromDiffSpec}
  n3 = VectorParameter() {parameterName=RotationalAxis, defaultValue=<rgba 1 0 0 1>}
  n4 = Subtract(B=VectorParameter() {parameterName=RotationalAxisPosition}.RGB)
  RotateAboutAxis(
   NormalizedRotationAxis=AppendVector(A=Normalize(VectorInput=@n3.RGB), B=ScalarParameter() {parameterName=RotationalAngle})
   RotationAngle=@n4
   PivotPoint=@n4
   Position=WorldPosition().XYZ
  )
  TextureSample(UVs=@n1) {samplerType=SAMPLERTYPE_Normal, texture=/Engine/EngineMaterials/DefaultNormal}
 ]
}
