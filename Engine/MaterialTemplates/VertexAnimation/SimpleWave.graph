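# Engine template material for simple wave vertex animation, using Default Lit shading with tiled default diffuse and normal textures.
# World position divided by a wave size is panned by time along a normalized direction, then passed through a sine remapped to 0–1.
# The wave is offset, then scaled by wind strength and a WaveStrength parameter.
# Constants supply wind direction, strength and speed; ConvertFromDiffSpec converts diffuse and specular inputs.
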Material /Engine/MaterialTemplates/VertexAnimation/SimpleWave
graph {
 properties {pixelDepthOffsetMode=PDOM_Legacy, usedShadingModels=Default Lit}
 nodes [
  n1 = TextureCoordinate() {uTiling=6, vTiling=6}
  n2 = Constant() {r=0.25}
  MaterialFunctionCall(
   DiffuseColor=LinearInterpolate(A=TextureSample(UVs=@n1) {texture=/Engine/EngineMaterials/DefaultDiffuse}.RGB, B=@n2, Alpha=@n2)
   SpecularColor=Multiply(A=Constant() {r=1}) {constA=1, constB=0.04, desc=FresnelBaseReflectFraction}
  ) {materialFunction=&ConvertFromDiffSpec}
  n3 = Constant() {r=0.5}
  n4 = ComponentMask(None=@n3) {b=true, g=true, r=true}
  n5 = Normalize(VectorInput=@n4)
  n6 = Multiply(A=Time(), B=ComponentMask(None=@n3) {a=true})
  n7 = Constant() {r=-1}
  n8 = Add(A=Divide(A=WorldPosition().XYZ, B=Constant() {desc=WaveSize, r=256}), B=Multiply(A=Multiply(A=@n6, B=@n7), B=@n5)) {desc=Pan time through space}
  n9 = ConstantBiasScale(None=Sine(None=Multiply(A=@n5, B=@n8) {desc=Cancel out waves not in line with wind direction}) {desc=Make a Wave}) {desc=Go from 0 to 1 instead of -1 to 1}
  n10 = Multiply(A=Distance(A=@n9, B=Constant()), B=Constant3Vector() {constant=<rgba 1 1 1 0>})
  n11 = Constant()
  Multiply(
   A=Multiply(A=Add(A=@n9, B=@n10), B=Distance(A=@n4, B=@n11)) {desc=Scale the Wave by WindStrength}
   B=ScalarParameter() {defaultValue=4, parameterName=WaveStrength}
  ) {desc=Amplitude}
  TextureSample(UVs=@n1) {samplerType=SAMPLERTYPE_Normal, texture=/Engine/EngineMaterials/DefaultNormal}
 ]
}
