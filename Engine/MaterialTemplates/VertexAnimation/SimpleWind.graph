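# Template Default Lit material for simple wind sway through vertex animation, with default diffuse and normal textures as the surface.
# Two sine waves panned through world position along the wind direction combine with a slow squared-sine gust.
# The result scales by wind strength, vertex color and a WaveStrength parameter.
# Wind direction, strength and fine-wave speed derive from a constant left at its default, apparently a placeholder.
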Material /Engine/MaterialTemplates/VertexAnimation/SimpleWind
graph {
 properties {bUsedWithStaticLighting=true, pixelDepthOffsetMode=PDOM_Legacy, usedShadingModels=Default Lit}
 nodes [
  Constant3Vector() {constant=<rgba 0 0 1 0>}
  n1 = TextureCoordinate() {uTiling=6, vTiling=6}
  n2 = Constant() {r=0.25}
  MaterialFunctionCall(
   DiffuseColor=LinearInterpolate(A=TextureSample(UVs=@n1) {texture=/Engine/EngineMaterials/DefaultDiffuse}.RGB, B=@n2, Alpha=@n2)
   SpecularColor=Multiply(A=Constant() {r=1}) {constA=1, constB=0.04, desc=FresnelBaseReflectFraction}
  ) {materialFunction=&ConvertFromDiffSpec}
  n3 = Constant()
  n4 = ComponentMask(None=@n3) {b=true, g=true, r=true}
  n5 = Normalize(VectorInput=@n4)
  n6 = WorldPosition()
  n7 = Constant() {desc=WaveSize, r=500}
  n8 = Time()
  n9 = Constant() {r=0.5}
  n10 = Multiply(A=Multiply(A=@n8, B=ComponentMask(None=@n9) {a=true}), B=Constant() {r=-0.25})
  Multiply(A=@n5, B=Add(A=Divide(A=@n6.XYZ, B=@n7), B=Multiply(A=@n10, B=@n5)) {desc=Pan time through space}) {desc=Cancel out waves not in line with wind direction}
  n11 = Divide(A=WorldPosition().XYZ, B=Constant() {desc=WaveSize, r=100})
  n12 = ComponentMask(None=@n3) {a=true}
  n13 = Multiply(A=Multiply(A=Multiply(A=@n8, B=@n12), B=Constant() {r=-1}), B=@n5)
  n14 = Sine(None=Multiply(A=@n5, B=Add(A=@n11, B=@n13) {desc=Pan time through space}) {desc=Cancel out waves not in line with wind direction}) {desc=Make a Wave}
  n15 = VertexColor() {desc=SmallerWaves * VertexRed}
  n16 = Divide(A=WorldPosition().XYZ, B=Constant() {desc=WaveSize, r=61})
  n17 = Sine(None=Multiply(A=@n5, B=Add(A=@n16, B=@n13) {desc=Pan time through space}) {desc=Cancel out waves not in line with wind direction}) {desc=Make a Wave}
  n18 = Add(
   A=Multiply(A=ConstantBiasScale(None=@n14) {desc=Go from 0 to 1 instead of -1 to 1}, B=Constant() {r=2})
   B=Multiply(A=@n15, B=ConstantBiasScale(None=@n17) {desc=Go from 0 to 1 instead of -1 to 1})
  )
  n19 = Distance(A=@n18, B=Constant())
  n20 = Constant3Vector() {constant=<rgba 1 1 1 0>}
  n21 = Power(Base=ConstantBiasScale(None=Sine(None=@n10) {desc=Make a Wave}), Exp=Constant() {r=2})
  n22 = Constant() {desc=MacroWaveAmplitude, r=10}
  n23 = Add(A=Add(A=@n18, B=Multiply(A=@n19, B=@n20)), B=Multiply(A=Multiply(A=@n21, B=@n22), B=@n5))
  n24 = Distance(A=@n4, B=Constant())
  Multiply(
   A=Multiply(A=Multiply(A=@n23, B=@n24) {desc=Scale the Wave by WindStrength}, B=@n15) {desc=Wind*VertexAlpha}
   B=ScalarParameter() {defaultValue=4, parameterName=WaveStrength}
  ) {desc=Amplitude}
  TextureSample(UVs=@n1) {samplerType=SAMPLERTYPE_Normal, texture=/Engine/EngineMaterials/DefaultNormal}
 ]
}
